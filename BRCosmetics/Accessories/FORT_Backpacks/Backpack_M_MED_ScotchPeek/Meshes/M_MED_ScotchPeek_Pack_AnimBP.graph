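# Animation Blueprint for the ScotchPeek backpack cosmetic, driving its idle animation on the pack skeleton.
# A single state machine feeds the output pose, starting in an idle state and alternating with an idle-break state.
# Each state plays a non-looping sequence, and automatic sequence-player-based transitions hand off to the other state, forming an endless cycle.
# The event graph holds only unconnected default nodes, and there are no variables.
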AnimBlueprint /BRCosmetics/Accessories/FORT_Backpacks/Backpack_M_MED_ScotchPeek/Meshes/M_MED_ScotchPeek_Pack_AnimBP
target_skeleton=&/M_MED_ScotchPeek_Pack_Skeleton
graphs [
 AnimGraph {nodes [
   Input Pose()
   AnimGraphNode_StateMachine_0 = New State Machine() {is=title AnimGraphNode_StateMachine}
   Output Pose(Result=@AnimGraphNode_StateMachine_0) {layerGroup=None}
  ]}
 New State Machine {
  entry=ScotchPeek_Idle_Part
  path=AnimGraph.AnimGraphNode_StateMachine_0.New State Machine
  ids=id
  nodes [
   AnimStateNode_0 = ScotchPeek_Idle_Part() {is=title AnimStateNode}
   AnimStateNode_1 = ScotchPeek_IdleBreak_Part() {is=title AnimStateNode}
   AnimStateTransitionNode_0 = AnimStateTransitionNode() {bAllowInertializationForSelfTransitions=true, bAutomaticRuleBasedOnSequencePlayerInState=true}
   AnimStateTransitionNode_1 = AnimStateTransitionNode() {bAllowInertializationForSelfTransitions=true, bAutomaticRuleBasedOnSequencePlayerInState=true}
  ]
  links [
   {
    dst=ScotchPeek_Idle_Part
    rule=AnimGraph.AnimGraphNode_StateMachine_0.New State Machine.AnimStateTransitionNode_1.Transition
    src=ScotchPeek_IdleBreak_Part
    via=AnimStateTransitionNode_1
   }
   {
    dst=ScotchPeek_IdleBreak_Part
    rule=AnimGraph.AnimGraphNode_StateMachine_0.New State Machine.AnimStateTransitionNode_0.Transition
    src=ScotchPeek_Idle_Part
    via=AnimStateTransitionNode_0
   }
  ]
 }
 ScotchPeek_Idle_Part {
  path=AnimGraph.AnimGraphNode_StateMachine_0.New State Machine.AnimStateNode_0.ScotchPeek_Idle_Part
  nodes [Output Animation Pose(Result=Sequence Player 'ScotchPeek_Idle_Part'() {bLoopAnimation=false, sequence=/BRCosmetics/Accessories/FORT_Backpacks/Backpack_M_MED_ScotchPeek/Anim/ScotchPeek_Idle_Part})]
 }
 ScotchPeek_IdleBreak_Part {
  path=AnimGraph.AnimGraphNode_StateMachine_0.New State Machine.AnimStateNode_1.ScotchPeek_IdleBreak_Part
  nodes [Output Animation Pose(Result=Sequence Player 'ScotchPeek_IdleBreak_Part'() {bLoopAnimation=false, sequence=/BRCosmetics/Accessories/FORT_Backpacks/Backpack_M_MED_ScotchPeek/Anim/ScotchPeek_IdleBreak_Part})]
 }
 Transition {path=AnimGraph.AnimGraphNode_StateMachine_0.New State Machine.AnimStateTransitionNode_0.Transition, nodes [Result()]}
 Transition {path=AnimGraph.AnimGraphNode_StateMachine_0.New State Machine.AnimStateTransitionNode_1.Transition, nodes [Result()]}
 EventGraph {nodes [
   TryGetPawnOwner()
   Event BlueprintUpdateAnimation()
  ]}
]
variables []
